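# ComponentWise_SphereMask is a library material function that applies a sphere mask to each channel of a vector independently.
# It splits the A, B, Radius and Hardness vector inputs into their components with BreakOutFloat3Components.
# Three SphereMask nodes each evaluate one matching channel, so every channel has its own center, radius and hardness.
# AppendMany recombines the three results into a single RGB output.
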MaterialFunction /Engine/Functions/Engine_MaterialFunctions02/Masking/ComponentWise_SphereMask
graph {
 properties {bExposeToLibrary=true, description=Apply a sphere mask for each channel of a vector independently.}
 nodes [
  n1 = MaterialFunctionCall(Float3=FunctionInput() {bUsePreviewValueAsDefault=true, inputName=A}) {materialFunction=&BreakOutFloat3Components}
  n2 = MaterialFunctionCall(Float3=FunctionInput() {bUsePreviewValueAsDefault=true, inputName=B, sortPriority=1}) {materialFunction=&BreakOutFloat3Components}
  n3 = MaterialFunctionCall(Float3=FunctionInput() {bUsePreviewValueAsDefault=true, inputName=Radius, sortPriority=2}) {materialFunction=&BreakOutFloat3Components}
  n4 = MaterialFunctionCall(Float3=FunctionInput() {bUsePreviewValueAsDefault=true, inputName=Hardness, sortPriority=3}) {materialFunction=&BreakOutFloat3Components}
  FunctionOutput(None=MaterialFunctionCall(
    R=SphereMask(A=@n1.R, B=@n2.R, Radius=@n3.R, Hardness=@n4.R)
    G=SphereMask(A=@n1.G, B=@n2.G, Radius=@n3.G, Hardness=@n4.G)
    B=SphereMask(A=@n1.B, B=@n2.B, Radius=@n3.B, Hardness=@n4.B)
   ) {materialFunction=&AppendMany}.RGB)
 ]
}
